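# Physics asset for the M_MED_Tour_Bus_Backpack back accessory, defining collision shapes and joints for its root and dangling belt bones.
# The backpack_root body combines a small capsule with a large rotated box, while dyn_belt_01 and dyn_belt_02 each use a small box.
# Angle-limited constraints chain dyn_belt_02 to dyn_belt_01 and dyn_belt_01 to the root, with matching swing limits and narrow twist on both joints.
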PhysicsAsset /BRCosmetics/Accessories/FORT_Backpacks/M_MED_Tour_Bus_Backpack/Meshes/M_MED_Tour_Bus_Backpack_Physics
bodies {
 backpack_root [
  {shapeType=Capsule, length=1, radius=1}
  {shapeType=Box, extentX=15.207, extentY=6.7616, extentZ=42.9855, center=<-1.8606 -4.4328 -2.6292>, rotation=<rot 53.8046 5.0011 0>}
 ]
 dyn_belt_01 [{shapeType=Box, extentX=3, extentY=1, extentZ=3, center=<0.5 0 0>}]
 dyn_belt_02 [{shapeType=Box, extentX=4, extentY=2, extentZ=4, center=<0.5 0 0>}]
}
constraints [
 {bone1Name=dyn_belt_01, bone2Name=backpack_root, swing1LimitDegrees=30, swing2LimitDegrees=15, twistLimitDegrees=2.5}
 {bone1Name=dyn_belt_02, bone2Name=dyn_belt_01, swing1LimitDegrees=30, swing2LimitDegrees=15, twistLimitDegrees=2.5}
]
